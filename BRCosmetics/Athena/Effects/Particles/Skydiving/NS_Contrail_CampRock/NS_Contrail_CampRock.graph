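# Campfire-themed skydiving contrail cosmetic that trails a bonfire, flames, sparks, hand trails and tumbling marshmallows from the player.
# Fire and trail emitters follow skeletal mesh bones, with a particle light and time-of-day bonfire tint.
# ParaGlideLeanAlpha blends intensity, size and speed throughout, while bIsFrontEndPreview retunes values for preview.
# Some marshmallows switch visibility and start smoke fades via a per-particle DoOnce comparing travel distance to PlaneHeight, with camera-angle-dependent scaling.
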NiagaraSystem /BRCosmetics/Athena/Effects/Particles/Skydiving/NS_Contrail_CampRock/NS_Contrail_CampRock
dependencies {
 usedDataInterfaces [
  &NiagaraDataInterfaceCurve
  &NiagaraDataInterfaceVector2DCurve
  &NiagaraDataInterfaceArrayFloat3
  &NiagaraDataInterfaceArrayInt32
  &NiagaraDataInterfaceMeshRendererInfo
  &NiagaraDataInterfaceArrayFloat
  &NiagaraDataInterfaceVectorCurve
  &NiagaraDataInterfaceCamera
  &NiagaraDataInterfaceColorCurve
 ]
 usedDynamicInputs [
  &SimulationPosition
  &MakeFloatFromVector
  /Niagara/DynamicInputs/CascadeConversion/SplitVector
 ]
 usedModules [&SystemState]
}
emitters {
 Bonfire {
  emitterUpdateScript [
   &EmitterState()
   &SpawnBurst_Instantaneous(Use Spawn Probability=false, UseLoopCountLimit=true) {module=SpawnBurst_Instantaneous001}
  ]
  particleSpawnScript [
   &InitializeParticle(Mesh Scale Mode=::NewEnumerator3, Mesh Uniform Scale=1.25, Position Mode=::NewEnumerator0, Write Material Random=true) {module=InitializeParticle001}
   &InitialMeshOrientation(Mesh Orientation Mode=::NewEnumerator2, Use Rotation Vector=false) {module=InitialMeshOrientation001}
  ]
  particleUpdateScript [
   &SkeletalMeshLocation(
    Bone Sampling Mode=::NewEnumerator4
    MaskSpawnGroup=false
    Orientation Sampling=::NewEnumerator1
    Sampled Position Offset=<0 0 -40>
    Sampled Position Offset Coordinate Space:ENiagaraSkelMeshTransforms=::NewEnumerator2
    Skeletal Mesh=$User.SkeletalMesh
   )
   &ParticleState(Kill Particles When Lifetime Has Elapsed=false)
   &ScaleMeshSize(Scale Factor=&VectorFromCurve(
     VectorCurve=DataInterface() {propertyValues {XCurve {keys [<at 0 0.5>,<at 1 1>]}, YCurve {keys [<at 0 0.5>,<at 1 1>]}, ZCurve {keys [<at 0 0.25>,<at 1 1>]}}}
     CurveIndex=$Emitter.Age
    )) {module=ScaleMeshSize002}
   &ScaleMeshSize(Scale Factor=&Lerp_Vector(
     A=&Lerp_Vector(
      A=<1.5 1.5 0.1>
      B=<1.33 1.33 0.1>
      Alpha=&Abs_Float(Float:NiagaraFloat=&DotProduct(Vector A=<0 0 1>, Vector B=&ReturnCameraVector(Camera Vectors=::NewEnumerator1, Coordinate Space=::World)))
     )
     B=<1.75 1.75 1>
     Alpha=$User.ParaGlideLeanAlpha
    )) {module=ScaleMeshSize001}
   &Color(Color=&ColorFromCurve(
     ColorCurve=DataInterface() {propertyValues {
       AlphaCurve {keys [<at 0 0.75 RCIM_Cubic>]}
       BlueCurve {keys [<at 0 0 RCIM_Cubic>,<at 1 0 RCIM_Cubic>]}
       GreenCurve {keys [<at 0 0.3642 RCIM_Cubic>,<at 1 0.4998 RCIM_Cubic>]}
       RedCurve {keys [<at 0 5 RCIM_Cubic>,<at 1 0.75 RCIM_Cubic>]}
      }}
     CurveIndex=&FNTimeOfDayScalar(Night Scale=1, Morning Scale=0, Day Scale=0, Evening Scale=1)
    )) {module=Color001}
   &FN_Light_Attributes(
    Color=<rgba 4 0.1908 0 1>
    Create Color Attribute=true
    Create Exponent Attribute=true
    Create Position Attribute=true
    Create Radius Attribute=true
    Enable Particle Light Enable=true
    Exponent=3
    Light Enabled=true
    Position=&AddVectorToPosition(
     Position=$Output.SkeletalMeshLocation.SampledPositionWithoutOffset
     Vector=&RotateVectorByQuaternion(Vector To Rotate=<50 50 0>, Quaternion=$Particles.SkeletalMeshLocation.SampledRotation)
    )
    Radius=&MakeCustomFloatFromBool(Bool=$User.bIsFrontEndPreview, True Float=120, False Float=240)
    Scale Color=&Multiply_Float(
     A=&Lerp_Float(A=0.8, B=2.5, Alpha=$User.ParaGlideLeanAlpha)
     B=&MakeCustomFloatFromBool(Bool=$User.bIsFrontEndPreview, True Float=25, False Float=50)
    )
   )
  ]
  rendererClasses [&NiagaraMeshRendererProperties, &NiagaraLightRendererProperties]
  renderers [
   {rendererClass=&NiagaraMeshRendererProperties, rendererIndex=0}
   {rendererClass=&NiagaraLightRendererProperties, rendererIndex=1}
  ]
 }
 FireGlow {
  emitterUpdateScript [
   &EmitterState()
   &SpawnBurst_Instantaneous(Use Spawn Probability=false, UseLoopCountLimit=true)
  ]
  particleSpawnScript [
   &InitializeParticle(
    Color=<rgba 1 0.3595 0.0468 0.25>
    Sprite Size Mode=::NewEnumerator3
    Uniform Sprite Size=200
    UsePositionOffset=false
    Write Material Random=true
   )
   &DynamicMaterialParameters(
    Index 0 Param 1=3
    Index 0 Param 2=10
    Index 0 Param 3=0.5
    Index 0 Param 4=1
    Module.Index0Float0Write=true
    Module.Index0Float1Write=true
    Module.Index0Float2Write=false
    Module.Index0Float3Write=true
   )
  ]
  particleUpdateScript [
   &SkeletalMeshLocation(
    Bone Sampling Mode=::NewEnumerator4
    MaskSpawnGroup=false
    Sampled Position Offset Coordinate Space:ENiagaraSkelMeshTransforms=::NewEnumerator2
    Skeletal Mesh=$User.SkeletalMesh
   )
   &ParticleState(Kill Particles When Lifetime Has Elapsed=false, Let Infinitely Lived Particles Die When Emitter Deactivates=true)
   &ScaleSpriteSize(
    Scale Sprite Size Mode=::NewEnumerator0
    Uniform Scale Factor=&Lerp_Float(A=1, B=1.25, Alpha=$User.ParaGlideLeanAlpha)
   )
   &Color(Color=&Multiply_LinearColorByFloat(
     Linear Color=&LerpLinearColors(StartColor=<rgba 1 0.2226 0 0.4>, EndColor=<rgba 2 0.3023 0 0.55>, LerpFactor=$User.ParaGlideLeanAlpha)
     Float=&MakeCustomFloatFromBool(Bool=$User.bIsFrontEndPreview, True Float=0.52, False Float=1)
    ))
   &CameraOffset(Camera Offset Amount=-10)
  ]
  rendererClasses [&NiagaraSpriteRendererProperties]
  renderers [{rendererClass=&NiagaraSpriteRendererProperties}]
 }
 Flame {
  emitterUpdateScript [
   &EmitterState()
   &SpawnBurst_Instantaneous(Spawn Count=5, Use Spawn Probability=false, UseLoopCountLimit=true) {module=SpawnBurst_Instantaneous002}
  ]
  particleSpawnScript [
   &InitializeParticle(
    Color=<rgba 5 0.5138 0 1>
    Mesh Renderer Array Visibility Mode=::NewEnumerator1
    Mesh Renderer Info=DataInterface() {propertyValues {MeshRenderer=&self:Flame_0.NiagaraMeshRendererProperties_0}}
    Mesh Renderer Mesh Index:NiagaraInt32=&SelectIntFromArray(
     Direct Array Index=&ReturnExecIndex()
     Int Selection Array=DataInterface() {propertyValues {IntData [0,0,0,0,1]}}
    )
    Mesh Scale=&SelectVectorFromArray(
     Direct Array Index=&ReturnExecIndex()
     Vector Selection Array=DataInterface() {propertyValues {FloatData [<0.35 0.35 1>, <0.35 0.35 1>, <0.35 0.35 1>, <0.35 0.35 1>, <0.5 0.5 1>]}}
    )
    Mesh Scale Mode=::NewEnumerator4
    UsePositionOffset=false
    Write Material Random=true
   ) {module=InitializeParticle001}
   &InitialMeshOrientation(
    Mesh Axis To Orient=<0 0 1>
    Mesh Orientation Mode=::NewEnumerator2
    Orientation Vector=<0 0 1>
    Rotation=&SelectVectorFromArray(
     Direct Array Index=&ReturnExecIndex()
     Vector Selection Array=DataInterface() {propertyValues {FloatData [<0 0 0>, <0 0 0.25>, <0 0 0.45>, <0 0 0.8>, <0 0 0>]}}
    )
    Use Rotation Vector=true
   )
  ]
  particleUpdateScript [
   &SkeletalMeshLocation(
    Bone Sampling Mode=::NewEnumerator4
    MaskSpawnGroup=false
    Orientation Sampling=::NewEnumerator1
    Sampled Position Offset=&Lerp_Vector(
     A=&SelectVectorFromArray(
      Direct Array Index=&ReturnExecIndex()
      Vector Selection Array=DataInterface() {propertyValues {FloatData [<0 0 -40>, <0 0 -40>, <0 0 -40>, <0 0 -40>, <0 0 -40>]}}
     )
     B=&SelectVectorFromArray(
      Direct Array Index=&ReturnExecIndex()
      Vector Selection Array=DataInterface() {propertyValues {FloatData [<20 10 -60>, <-20 10 -60>, <-20 -10 -60>, <20 -10 -60>, <0 0 -40>]}}
     )
     Alpha=$User.ParaGlideLeanAlpha
    )
    Sampled Position Offset Coordinate Space:ENiagaraSkelMeshTransforms=::NewEnumerator2
    Skeletal Mesh=$User.SkeletalMesh
   ) {module=SkeletalMeshLocation002}
   &DynamicMaterialParameters(
    Index 0 Param 1=&SelectFloatFromArray(
     Direct Array Index=$Particles.PartitionParticles.Partition
     Float Selection Array=DataInterface() {propertyValues {FloatData [0.8,0.6,0.2,0.1,0.5]}}
    )
    Index 0 Param 4=0.6
    Index 1 Param 4=1
    Index 3 Param 1=&Lerp_Float(A=4, B=1, Alpha=$User.ParaGlideLeanAlpha)
    Index 3 Param 2=$User.ParaGlideLeanAlpha
    Module.Param0WriteEnabled=true
    Module.Param0WriteEnabled1=false
    Module.Param0WriteEnabled2=true
    Module.Param0WriteEnabled3=true
    Module.Param1WriteEnabled=false
    Module.Param1WriteEnabled1=false
    Module.Param1WriteEnabled2=false
    Module.Param1WriteEnabled3=true
    Module.Param2WriteEnabled=false
    Module.Param2WriteEnabled1=false
    Module.Param2WriteEnabled2=false
    Module.Param2WriteEnabled3=true
    Module.Param3WriteEnabled=true
    Module.Param3WriteEnabled1=true
    Module.Param3WriteEnabled2=false
    Module.Param3WriteEnabled3=true
    Write Parameter Index 1=true
    Write Parameter Index 2=true
    Write Parameter Index 3=true
   ) {module=DynamicMaterialParameters001}
   &ParticleState(Kill Particles When Lifetime Has Elapsed=false, Let Infinitely Lived Particles Die When Emitter Deactivates=true)
   &ScaleMeshSize(Scale Factor=&VectorFromCurve(
     VectorCurve=DataInterface() {propertyValues {XCurve {keys [<at 0 0.25>,<at 1 1>]}, YCurve {keys [<at 0 0.25>,<at 1 1>]}, ZCurve {keys [<at 0 0>,<at 1 1>]}}}
     CurveIndex=$Emitter.Age
    )) {module=ScaleMeshSize001}
   &ScaleMeshSize(Scale Factor=&Lerp_Vector(
     A=&Lerp_Vector(
      A=<1 1 1>
      B=&MakeVector(
       X=1.5
       Y=1.25
       Z=&Waveform(
        "Waveform [1]"=::NewEnumerator0
        "[1] Phase Scale"=&RandomRangeFloat(Minimum=0.5, Maximum=1)
        "[1] Frequency"=&RandomRangeFloat(Minimum=0.35, Maximum=1)
        "[1] Amplitude Min/Max"=&RandomRangeVector2D(Minimum=<0.5 1>, Maximum=<0.5 1.25>)
       )
      )
      Alpha=$User.ParaGlideLeanAlpha
     )
     B=&MakeVector(
      X=1.5
      Y=1.5
      Z=&Waveform(
       "Waveform [1]"=::NewEnumerator0
       "[1] Phase Scale"=&RandomRangeFloat(Minimum=0.5, Maximum=1)
       "[1] Frequency"=&RandomRangeFloat(Minimum=0.35, Maximum=1)
       "[1] Amplitude Min/Max"=&RandomRangeVector2D(Minimum=<0.5 1>, Maximum=<0.5 1.25>)
       "[1] Scale"=2
      )
     )
     Alpha=$User.ParaGlideLeanAlpha
    ))
   &SolveForcesAndVelocity(Clamp Velocity=false, Limit Acceleration=false)
   &ScaleColor(
    Scale Alpha=&Lerp_Float(
     A=&MakeCustomFloatFromBool(
      Bool=$User.bIsFrontEndPreview
      True Float=&SelectFloatFromArray(
       Direct Array Index=&ReturnExecIndex()
       Float Selection Array=DataInterface() {propertyValues {FloatData [1,1,1,1,0.35]}}
      )
      False Float=&SelectFloatFromArray(
       Direct Array Index=&ReturnExecIndex()
       Float Selection Array=DataInterface() {propertyValues {FloatData [1,1,1,1,1.25]}}
      )
     )
     B=&SelectFloatFromArray(
      Direct Array Index=&ReturnExecIndex()
      Float Selection Array=DataInterface() {propertyValues {FloatData [1,1,1,1,0]}}
     )
     Alpha=$User.ParaGlideLeanAlpha
    )
    ScaleA=true
    ScaleRGB=false
   )
  ]
  rendererClasses [&NiagaraMeshRendererProperties]
  renderers [{rendererClass=&NiagaraMeshRendererProperties, rendererIndex=0}]
 }
 Hand_Trails {
  emitterUpdateScript [
   &EmitterState(
    Inactive Response=::NewEnumerator1
    Life Cycle Mode=::NewEnumerator1
    Loop Behavior=::NewEnumerator1
    Loop Duration Mode=::NewEnumerator1
    UseLoopDelay=false
   )
   &SpawnBurst_Instantaneous(Spawn Count=2, Use Spawn Probability=false, UseLoopCountLimit=false)
  ]
  particleSpawnScript [
   &InitializeParticle(Color=<rgba 1 1 1 0>, Mesh Scale=<1 1 5>, Mesh Scale Mode=::NewEnumerator4, UsePositionOffset=false, Write Material Random=true)
   &CameraOffset(Camera Offset Amount=10)
  ]
  particleUpdateScript [
   &SkeletalMeshLocation(
    Bone / Socket Index=&SelectIntFromArray(Direct Array Index=&ReturnExecIndex(), Int Selection Array=DataInterface() {propertyValues {IntData [1,2]}})
    Bone Sampling Mode=::NewEnumerator4
    Inherit Velocity=false
    Interpolate Mesh Sampling=false
    MaskSpawnGroup=false
    Skeletal Mesh=$User.SkeletalMesh
   )
   &ParticleState(Kill Particles When Lifetime Has Elapsed=false, Let Infinitely Lived Particles Die When Emitter Deactivates=true)
   &ScaleMeshSize(Scale Factor=&Lerp_Vector(A=<1 1 1>, B=<10 10 2>, Alpha=$User.ParaGlideLeanAlpha)) {module=ScaleMeshSize003}
   &SolveForcesAndVelocity(Clamp Velocity=false, Limit Acceleration=false)
   &DynamicMaterialParameters(
    Index 0 Param 1=&RandomRangeFloat(Minimum=0.5, Maximum=1)
    Index 0 Param 2=&Lerp_Float(
     A=&RandomRangeFloat(Minimum=1.25, Maximum=2)
     B=&RandomRangeFloat(Minimum=2, Maximum=3)
     Alpha=$User.ParaGlideLeanAlpha
    )
    Index 0 Param 3=&RandomRangeFloat(Minimum=0.5, Maximum=1)
    Index 0 Param 4=&RandomRangeFloat(Minimum=0.5, Maximum=1)
    Index 3 Param 1=&RandomRangeFloat(Minimum=5, Maximum=10)
    Index 3 Param 2=&Lerp_Float(
     A=&RandomRangeFloat(Minimum=15, Maximum=25)
     B=&RandomRangeFloat(Minimum=35, Maximum=45)
     Alpha=$User.ParaGlideLeanAlpha
    )
    Index 3 Param 3=&RandomRangeFloat(Minimum=3, Maximum=5)
    Index 3 Param 4=&RandomRangeFloat(Minimum=0.5, Maximum=1)
    Module.Index0Float0Write=true
    Module.Index0Float1Write=true
    Module.Index0Float2Write=true
    Module.Index0Float3Write=true
    Module.Index3Float0Write=true
    Module.Index3Float1Write=true
    Module.Index3Float2Write=true
    Module.Index3Float3Write=true
    Write Parameter Index 3=::NewEnumerator1
   ) {module=DynamicMaterialParameters002}
   &Color(
    Color=&LerpLinearColors(
     StartColor=&MakeCustomLinearColorFromBool(Bool=$User.bIsFrontEndPreview, True Color=<rgba 0.0417 0.0417 0.0417 1>, False Color=<rgba 0.2396 0.2396 0.2396 1>)
     EndColor=<rgba 20 5.0991 0 1>
     LerpFactor=$User.ParaGlideLeanAlpha
    )
    Scale Alpha=0.1
   )
  ]
  rendererClasses [&NiagaraMeshRendererProperties]
  renderers [{rendererClass=&NiagaraMeshRendererProperties, rendererIndex=0}]
 }
 Marshmallow {
  emitterUpdateScript [
   &EmitterState(
    Life Cycle Mode=::NewEnumerator1
    Loop Duration=&Lerp_Float(A=0.6, B=0.5, Alpha=$User.ParaGlideLeanAlpha)
    UseLoopDelay=false
   )
   &SpawnBurst_Instantaneous(
    Spawn Count=&RandomRangeInt(Minimum=1, Maximum=4)
    Use Spawn Probability=false
    UseLoopCountLimit=false
   )
   /SetVariables(Emitter.GrpID:NiagaraInt32=&RandomRangeInt(Minimum=0, Maximum=1000)) {isSetParameters=true, module=SetVariables_#1}
   /SetVariables(Emitter.Dir:Vector3f=&RandomVector(
     Vector Scale=1
     Constrain to Hemisphere=true
     Hemisphere Constraint Vector=<0 0 -1>
     Randomness Mode=::NewEnumerator1
     Random Seed=$Emitter.GrpID
     Fixed Random Seed=true
    )) {isSetParameters=true, module=SetVariables_#2}
  ]
  particleSpawnScript [
   &InitializeParticle(
    Lifetime Max=2.5
    Lifetime Min=2
    Lifetime Mode=::NewEnumerator1
    Mesh Scale Mode=::NewEnumerator1
    Mesh Uniform Scale Max=0.06
    Mesh Uniform Scale Min=0.04
    Sprite Size Mode=::NewEnumerator1
    Uniform Sprite Size Max=40
    Uniform Sprite Size Min=30
    UsePositionOffset=false
    Write Material Random=true
   )
   &InitialMeshOrientation(Use Rotation Vector=false)
   &SystemLocation(Offset=&RandomVector(Vector Scale=10))
   /SetVariables(Particles.Position=&MakeCustomVectorFromBool(
     Bool=$User.bIsFrontEndPreview
     True Vector=&Lerp_Vector(A=<0 35 0>, B=<0 35 -15>, Alpha=$User.ParaGlideLeanAlpha)
     False Vector=Linked() {linkedVariable {name=Engine.Emitter.SimulationPosition, type=NiagaraPosition}}
    )) {isSetParameters=true, module=SetVariables_#3}
   &PointForce(
    Force Falloff Distance=79.3811
    Force Strength=&RandomRangeFloat(Minimum=125, Maximum=180)
    Linear Force Vector:Vector3f=&RandomRangeVector(Minimum=<-1 -1 0>, Maximum=<1 1 0>)
    Module.Use Falloff=false
    Module.UseLinearForce=true
   )
   &AddVelocity(
    Cone Angle=150
    Cone Axis=$Emitter.Dir
    Module.UseConeFalloff=false
    Module.UseVelDistribution=false
    Override Seed=false
    Velocity Mode=::NewEnumerator2
    Velocity Speed=&Lerp_Float(
     A=&RandomRangeFloat(Minimum=150, Maximum=200)
     B=&RandomRangeFloat(Minimum=180, Maximum=220)
     Alpha=$User.ParaGlideLeanAlpha
    )
   )
   &ApplyInitialForces()
  ]
  particleUpdateScript [
   &ParticleState()
   &ScaleMeshSize(Scale Factor=&Lerp_Vector(
     A=&VectorFromFloat(Value=&FloatFromCurve(FloatCurve=DataInterface() {propertyValues {Curve {
          defaultValue=0
          keys [
           {arriveTangent=4.1785, interpMode=RCIM_Cubic, leaveTangent=8.3451, tangentMode=RCTM_Break, time=0, value=0}
           {arriveTangent=20, interpMode=RCIM_Cubic, leaveTangent=20, time=0.0309, value=1.5}
           {arriveTangent=-4.7186, leaveTangent=1.1179, time=0.05, value=1}
           {arriveTangent=-1.0526, leaveTangent=-1.0526, time=0.9, value=1.0462}
           <at 1 0 RCIM_Cubic>
          ]
         }}}))
     B=&Lerp_Vector(
      A=&VectorFromFloat(Value=&FloatFromCurve(FloatCurve=DataInterface() {propertyValues {Curve {
           defaultValue=0
           keys [{arriveTangent=-4.7186, interpMode=RCIM_Cubic, leaveTangent=-0.3024, tangentMode=RCTM_Break, time=0.55, value=1}, <at 0.6 0 RCIM_Cubic>]
          }}}))
      B=&VectorFromFloat(Value=&FloatFromCurve(FloatCurve=DataInterface() {propertyValues {Curve {
           defaultValue=0
           keys [{arriveTangent=-4.7186, interpMode=RCIM_Cubic, leaveTangent=-0.3024, tangentMode=RCTM_Break, time=0.4, value=1}, <at 0.42 0 RCIM_Cubic>]
          }}}))
      Alpha=$User.ParaGlideLeanAlpha
     )
     Alpha=&FloatFromCurve(
      FloatCurve=DataInterface() {propertyValues {Curve {keys [{interpMode=RCIM_Cubic, tangentMode=RCTM_User, time=0.95, value=0}, <at 0.99 1>]}}}
      CurveIndex=&Abs_Float(Float:NiagaraFloat=&DotProduct(Vector A=<0 0 1>, Vector B=&ReturnCameraVector(Camera Vectors=::NewEnumerator1, Coordinate Space=::World)))
     )
    ))
   &UpdateMeshOrientation(
    Rotation Rate=&RandomRangeFloat(Minimum=1.25, Maximum=1.75)
    Rotation Vector=&RandomRangeVector(Minimum=<-1 -1 -1>, Maximum=<1 1 1>)
   )
   &GravityForce(Gravity=&RandomRangeVector(Minimum=<0 0 500>, Maximum=<0 0 750>))
   &SolveForcesAndVelocity(Clamp Velocity=false, Limit Acceleration=false)
   /SetVariables(Particles.Velocity=&Lerp_Vector(
     A=$Particles.Velocity
     B=&Lerp_Vector(A=<0 0 200>, B=<0 0 500>, Alpha=$User.ParaGlideLeanAlpha)
     Alpha=&FloatFromCurve(
      FloatCurve=DataInterface() {propertyValues {Curve {defaultValue=1, keys [{interpMode=RCIM_Cubic, tangentMode=RCTM_Break, time=0, value=0}, {tangentMode=RCTM_User, time=3, value=1}]}}}
      CurveIndex=$Particles.Age
     )
    )) {isSetParameters=true, module=SetVariables_#4}
   &ScaleColor(
    Scale Alpha=&FloatFromCurve(
     FloatCurve=DataInterface() {propertyValues {Curve {keys [{interpMode=RCIM_Cubic, tangentMode=RCTM_Break, time=0, value=1}, {interpMode=RCIM_Cubic, tangentMode=RCTM_Break, time=1, value=0}]}}}
     CurveIndex=$Particles.NewNormalizedAge
    )
    ScaleA=true
    ScaleRGB=false
   )
  ]
  rendererClasses [&NiagaraMeshRendererProperties]
  renderers [{rendererClass=&NiagaraMeshRendererProperties, rendererIndex=0}]
 }
 Marshmallow_withToasted001 {
  emitterUpdateScript [
   &EmitterState(
    Life Cycle Mode=::NewEnumerator1
    Loop Duration=&Lerp_Float(A=0.5, B=0.4, Alpha=$User.ParaGlideLeanAlpha)
    UseLoopDelay=false
   )
   &SpawnBurst_Instantaneous(
    Spawn Count=&RandomRangeInt(Minimum=1, Maximum=2)
    Use Spawn Probability=false
    UseLoopCountLimit=false
   )
   /SetVariables(Emitter.GrpID:NiagaraInt32=&RandomRangeInt(Minimum=0, Maximum=1000)) {isSetParameters=true, module=SetVariables_#1}
   /SetVariables(Emitter.Dir:Vector3f=&RandomVector(
     Vector Scale=1
     Constrain to Hemisphere=true
     Hemisphere Constraint Vector=<0 0 -1>
     Randomness Mode=::NewEnumerator1
     Random Seed=$Emitter.GrpID
     Fixed Random Seed=true
    )) {isSetParameters=true, module=SetVariables_#2}
  ]
  particleSpawnScript [
   &InitializeParticle(
    Color Mode=::NewEnumerator0
    Lifetime Max=3
    Lifetime Min=2.5
    Lifetime Mode=::NewEnumerator1
    Mesh Scale Mode=::NewEnumerator1
    Mesh Uniform Scale Max=0.06
    Mesh Uniform Scale Min=0.04
    Position Mode=::NewEnumerator0
    Sprite Rotation Mode=::NewEnumerator2
    Sprite Size Mode=::NewEnumerator1
    Uniform Sprite Size Max=20
    Uniform Sprite Size Min=15
    Write Material Random=true
   )
   /SetVariables(
    Particles.ColorSprite:LinearColor=<rgba 0.0417 0.0417 0.0417 1>
    Particles.HitTime:NiagaraFloat=$Particles.Lifetime
    Particles.NewLifetime:NiagaraFloat=&Lerp_Float(
     A=&RandomRangeFloat(Minimum=0.45, Maximum=0.6)
     B=&RandomRangeFloat(Minimum=0.4, Maximum=0.45)
     Alpha=$User.ParaGlideLeanAlpha
    )
    Particles.ScaleSmokeSize:Vector3f=&RandomRangeVector(Minimum=<0.75 0.75 1.25>, Maximum=<0.85 0.85 1.5>)
    Particles.SmokeOrientation:Quat4f=<0 0 1 0>
   ) {isSetParameters=true}
   /SetVariables(
    Particles.DynamicMaterialParameter4:Vector4f=&MakeVector4(
     X=&RandomRangeFloat(Minimum=0.5, Maximum=1)
     Y=&RandomRangeFloat(Minimum=0.8, Maximum=1)
     Z=&RandomRangeFloat(Minimum=0.5, Maximum=1)
     W=&RandomRangeFloat(Minimum=0.5, Maximum=1)
    )
    Particles.DynamicMaterialParameter5:Vector4f=&MakeVector4(
     X=&RandomRangeFloat(Minimum=3, Maximum=8)
     Y=&MakeCustomFloatFromBool(
      Bool=$User.bIsFrontEndPreview
      True Float=&RandomRangeFloat(Minimum=15, Maximum=20)
      False Float=&RandomRangeFloat(Minimum=10, Maximum=15)
     )
     Z=&RandomRangeFloat(Minimum=2, Maximum=5)
     W=&RandomRangeFloat(Minimum=0.5, Maximum=1)
    )
   ) {isSetParameters=true}
   &InitialMeshOrientation(Use Rotation Vector=false)
   &SystemLocation(Offset=&RandomVector(Vector Scale=25))
   /SetVariables(Particles.Position=&MakeCustomVectorFromBool(
     Bool=$User.bIsFrontEndPreview
     True Vector=&Lerp_Vector(A=<0 35 0>, B=<0 35 -15>, Alpha=$User.ParaGlideLeanAlpha)
     False Vector=Linked() {linkedVariable {name=Engine.Emitter.SimulationPosition, type=NiagaraPosition}}
    )) {isSetParameters=true, module=SetVariables_#3}
   &AddVelocity(
    Cone Angle=&Lerp_Float(A=10, B=20, Alpha=$User.ParaGlideLeanAlpha)
    Cone Axis=$Emitter.Dir
    Module.UseConeFalloff=false
    Module.UseVelDistribution=false
    Override Seed=false
    Velocity Mode=::NewEnumerator2
    Velocity Speed=&Lerp_Float(
     A=&RandomRangeFloat(Minimum=55, Maximum=65)
     B=&RandomRangeFloat(Minimum=100, Maximum=120)
     Alpha=$User.ParaGlideLeanAlpha
    )
   )
   &PointForce(
    Force Falloff Distance=79.3811
    Force Strength=&RandomRangeFloat(Minimum=100, Maximum=150)
    Linear Force Vector:Vector3f=&RandomRangeVector(Minimum=<-1 -1 0>, Maximum=<1 1 0>)
    Module.Use Falloff=false
    Module.UseLinearForce=true
   )
   &ApplyInitialForces()
  ]
  particleUpdateScript [
   &ParticleState()
   &UpdateMeshOrientation(
    Rotation Rate=&RandomRangeFloat(Minimum=1.25, Maximum=1.75)
    Rotation Vector=&RandomRangeVector(Minimum=<-1 -1 -1>, Maximum=<1 1 1>)
   )
   &GravityForce(Gravity=<0 0 500>) {module=GravityForce001}
   &DoOnce(Trigger Condition:NiagaraBool=&SetBoolByFloatComparison(
     A=&DistanceBetweenPositions(A=$Engine.Emitter.SimulationPosition, B=$Particles.Position)
     B=$User.PlaneHeight
    ))
   &ScaleMeshSize(Scale Factor=&VectorFromFloat(Value=&FloatFromCurve(FloatCurve=DataInterface() {propertyValues {Curve {
         defaultValue=0
         keys [
          {arriveTangent=4.1785, interpMode=RCIM_Cubic, leaveTangent=8.3451, tangentMode=RCTM_Break, time=0, value=0}
          {arriveTangent=20, interpMode=RCIM_Cubic, leaveTangent=20, time=0.0309, value=1.5}
          {arriveTangent=-4.7186, interpMode=RCIM_Cubic, leaveTangent=1.1179, tangentMode=RCTM_Break, time=0.05, value=1}
         ]
        }}}))) {module=ScaleMeshSize001}
   &SolveForcesAndVelocity(Clamp Velocity=false, Limit Acceleration=false)
   /SetVariables(Particles.ScaleToasted:Vector3f=&Lerp_Vector(
     A=&VectorFromFloat(Value=&FloatFromCurve(
       FloatCurve=DataInterface() {propertyValues {Curve {
          defaultValue=0
          keys [
           {arriveTangent=-4.7186, interpMode=RCIM_Cubic, leaveTangent=-0.0031, tangentMode=RCTM_Break, time=0.05, value=1}
           {arriveTangent=-1.9528, interpMode=RCIM_Cubic, tangentMode=RCTM_Break, time=1, value=0}
          ]
         }}}
       Scale Curve=0.045
      ))
     B=&VectorFromFloat(Value=&Lerp_Float(
       A=&FloatFromCurve(
        FloatCurve=DataInterface() {propertyValues {Curve {
           defaultValue=0
           keys [{arriveTangent=-4.7186, interpMode=RCIM_Cubic, leaveTangent=-0.3024, tangentMode=RCTM_Break, time=0.5, value=1}, <at 0.54 0 RCIM_Cubic>]
          }}}
        Scale Curve=0.045
       )
       B=&FloatFromCurve(
        FloatCurve=DataInterface() {propertyValues {Curve {
           defaultValue=0
           keys [{arriveTangent=-4.7186, interpMode=RCIM_Cubic, leaveTangent=-0.3024, tangentMode=RCTM_Break, time=0.33, value=1}, <at 0.38 0 RCIM_Cubic>]
          }}}
        Scale Curve=0.045
       )
       Alpha=$User.ParaGlideLeanAlpha
      ))
     Alpha=&FloatFromCurve(
      FloatCurve=DataInterface() {propertyValues {Curve {keys [{interpMode=RCIM_Cubic, tangentMode=RCTM_User, time=0.95, value=0}, <at 0.99 1>]}}}
      CurveIndex=&Abs_Float(Float:NiagaraFloat=&DotProduct(Vector A=<0 0 1>, Vector B=&ReturnCameraVector(Camera Vectors=::NewEnumerator1, Coordinate Space=::World)))
     )
    )) {isSetParameters=true}
   /SetVariables(Particles.HitTime:NiagaraFloat=&MakeCustomFloatFromBool(Bool=$Particles.DoOnce.Execute, True Float=$Particles.Age, False Float=$Particles.HitTime)) {isSetParameters=true}
   /SetVariables(Particles.NewNormalizedAge:NiagaraFloat=&ClampFloat(Float=&Divide_Float(A=&Subtract_Float(A=$Particles.Age, B=$Particles.HitTime), B=$Particles.NewLifetime))) {isSetParameters=true}
   /SetVariables(Particles.VisibilityTag=&MakeCustomIntFromBool(Bool=$Particles.DoOnce.TriggeredOnce, True INT=1, False INT=0)) {isSetParameters=true}
   /SetVariables(Particles.Velocity=&Lerp_Vector(
     A=$Particles.Velocity
     B=&Lerp_Vector(A=<0 0 200>, B=<0 0 500>, Alpha=$User.ParaGlideLeanAlpha)
     Alpha=&FloatFromCurve(
      FloatCurve=DataInterface() {propertyValues {Curve {defaultValue=1, keys [{interpMode=RCIM_Cubic, tangentMode=RCTM_Break, time=0, value=0}, {tangentMode=RCTM_User, time=3, value=1}]}}}
      CurveIndex=$Particles.Age
     )
    )) {isSetParameters=true, module=SetVariables_#4}
   /SetVariables(Particles.ColorSmoke:LinearColor=&Multiply_LinearColorByFloat(
     Linear Color=&MakeCustomLinearColorFromBool(
      Bool=$User.bIsFrontEndPreview
      True Color=&ColorFromCurve(ColorCurve=DataInterface() {propertyValues {
         AlphaCurve {keys [<at 0 1 RCIM_Cubic>]}
         BlueCurve {keys [<at 0.2451 0 RCIM_Cubic>,<at 0.3926 1 RCIM_Cubic>]}
         GreenCurve {keys [<at 0.2451 19.924 RCIM_Cubic>,<at 0.3926 1 RCIM_Cubic>]}
         RedCurve {keys [<at 0.2451 75 RCIM_Cubic>,<at 0.3926 1 RCIM_Cubic>]}
        }})
      False Color=&ColorFromCurve(ColorCurve=DataInterface() {propertyValues {
         AlphaCurve {keys [<at 0 1 RCIM_Cubic>]}
         BlueCurve {keys [<at 0.2 0 RCIM_Cubic>,<at 0.405 0.9896 RCIM_Cubic>]}
         GreenCurve {keys [<at 0.2 265.6535 RCIM_Cubic>,<at 0.405 0.9896 RCIM_Cubic>]}
         RedCurve {keys [<at 0.2 1000 RCIM_Cubic>,<at 0.405 0.9896 RCIM_Cubic>]}
        }})
     )
     Float=0.25
    )) {isSetParameters=true}
   /SetVariables(Particles.SmokeOrientation:Quat4f=&MakeQuaternion(
     Quaternion From=0
     Axis=&Lerp_Vector(
      A=<0 0 1>
      B=<1 0 0>
      Alpha=&FloatFromCurve(
       FloatCurve=DataInterface() {propertyValues {Curve {keys [{interpMode=RCIM_Cubic, tangentMode=RCTM_User, time=0.5, value=0}, <at 0.9 1>]}}}
       CurveIndex=&Abs_Float(Float:NiagaraFloat=&DotProduct(Vector A=<0 0 1>, Vector B=&ReturnCameraVector(Camera Vectors=::NewEnumerator1, Coordinate Space=::World)))
      )
     )
     Angle=&FloatFromCurve(
      FloatCurve=DataInterface() {propertyValues {Curve {keys [{interpMode=RCIM_Cubic, tangentMode=RCTM_User, time=0.85, value=0}, <at 0.9 1>]}}}
      CurveIndex=&Abs_Float(Float:NiagaraFloat=&DotProduct(Vector A=<0 0 1>, Vector B=&ReturnCameraVector(Camera Vectors=::NewEnumerator1, Coordinate Space=::World)))
      Scale Curve=-0.1
     )
     From=?
     To=?
     Stabilization Mode:ENiagara_OrientVectorCount=::NewEnumerator0
     Side Direction:Vector3f=?
     Up Direction:Vector3f=?
     Angle Type=::NewEnumerator1
     Normalize:NiagaraBool=true
    )) {isSetParameters=true}
   /SetVariables(Particles.ScaleSmokeSize:Vector3f=&Multiply_VectorByFloat(
     Vector=&MakeVector(
      X=0.8
      Y=0.8
      Z=&Lerp_Float(
       A=1
       B=&Lerp_Float(
        A=&FloatFromCurve(FloatCurve=DataInterface() {propertyValues {Curve {keys [
             {interpMode=RCIM_Cubic, leaveTangent=-2.4219, tangentMode=RCTM_Break, time=0.45, value=1}
             {interpMode=RCIM_Cubic, tangentMode=RCTM_Break, time=0.5, value=0}
            ]}}})
        B=&FloatFromCurve(FloatCurve=DataInterface() {propertyValues {Curve {keys [{interpMode=RCIM_Cubic, tangentMode=RCTM_Break, time=0.3, value=1}, <at 0.35 0>]}}})
        Alpha=$User.ParaGlideLeanAlpha
       )
       Alpha=&FloatFromCurve(
        FloatCurve=DataInterface() {propertyValues {Curve {
           defaultValue=1
           keys [{interpMode=RCIM_Cubic, tangentMode=RCTM_Break, time=0.94, value=0}, {interpMode=RCIM_Cubic, tangentMode=RCTM_Break, time=0.99, value=1}]
          }}}
        CurveIndex=&Abs_Float(Float:NiagaraFloat=&DotProduct(Vector A=<0 0 1>, Vector B=&ReturnCameraVector(Camera Vectors=::NewEnumerator1, Coordinate Space=::World)))
       )
      )
     )
     Float=&RandomRangeFloat(Minimum=0.9, Maximum=1.1)
    )) {isSetParameters=true}
   &ScaleSpriteSize(
    Uniform Curve Index=$Particles.NewNormalizedAge
    Uniform Curve Scale=&MakeCustomFloatFromBool(
     Bool=&SetBoolByFloatComparison(A=$Particles.NewNormalizedAge, B=1, Comparison Type=::NewEnumerator6)
     True Float=&MakeCustomFloatFromBool(Bool=$User.bIsFrontEndPreview, True Float=1.125, False Float=1)
     False Float=0
    )
    Uniform Curve Sprite Scale=DataInterface() {propertyValues {Curve {
       defaultValue=0
       keys [
        {arriveTangent=0.3979, interpMode=RCIM_Cubic, leaveTangent=2.8266, tangentMode=RCTM_Break, time=0, value=0.25}
        {arriveTangent=2.2166, interpMode=RCIM_Cubic, leaveTangent=1.2933, tangentMode=RCTM_Break, time=1, value=1}
       ]
      }}}
   )
   &DynamicMaterialParameters(
    Index 0 Param 1=&Add_Float(
     A=&RandomRangeFloat(Minimum=0.3, Maximum=1)
     B=&FloatFromCurve(
      FloatCurve=DataInterface() {propertyValues {Curve {
         defaultValue=1
         keys [{interpMode=RCIM_Cubic, tangentMode=RCTM_Break, time=0, value=0}, {interpMode=RCIM_Cubic, tangentMode=RCTM_Break, time=1, value=1}]
        }}}
      CurveIndex=$Particles.NewNormalizedAge
      Scale Curve=&MakeCustomFloatFromBool(
       Bool=&SetBoolByFloatComparison(A=$Particles.NewNormalizedAge, B=1, Comparison Type=::NewEnumerator6)
       True Float=0.7
       False Float=0
      )
     )
    )
    Index 0 Param 3=&FloatFromCurve(
     FloatCurve=DataInterface() {propertyValues {Curve {keys [
         {interpMode=RCIM_Cubic, leaveTangent=-0.3098, tangentMode=RCTM_Break, time=0, value=1}
         {arriveTangent=0.2947, tangentMode=RCTM_User, time=1, value=0}
        ]}}}
     CurveIndex=$Particles.NewNormalizedAge
    )
    Index 0 Param 4=&FloatFromCurve(FloatCurve=DataInterface() {propertyValues {Curve {keys [<at 0 1>,<at 1 0>]}}}, CurveIndex=$Particles.NewNormalizedAge)
    Index 1 Param 4=1
    Index 2 Param 1=&FloatFromCurve(
     FloatCurve=DataInterface() {propertyValues {Curve {keys [
         {arriveTangent=4.9626, interpMode=RCIM_Cubic, leaveTangent=2.5739, tangentMode=RCTM_User, time=0, value=0}
         {leaveTangent=-0.0719, time=1, value=1}
        ]}}}
     Scale Curve=3
    )
    Index 2 Param 2=&FloatFromCurve(FloatCurve=DataInterface() {propertyValues {Curve {keys [<at 0 0.5>,<at 1 1>]}}})
    Index 2 Param 3=&FloatFromCurve(
     FloatCurve=DataInterface() {propertyValues {Curve {keys [
         {interpMode=RCIM_Cubic, leaveTangent=-6.6324, leaveTangentWeight=0.1676, tangentMode=RCTM_Break, time=0, value=1}
         {interpMode=RCIM_Cubic, tangentMode=RCTM_User, time=0.2, value=0}
         {interpMode=RCIM_Cubic, leaveTangent=11.1233, tangentMode=RCTM_Break, time=1, value=1}
        ]}}}
     Scale Curve=5
    )
    Index 2 Param 4=&RandomRangeFloat(Minimum=0.7, Maximum=1)
    Index 3 Param 1=&RandomRangeFloat(Minimum=5, Maximum=10)
    Index 3 Param 2=&RandomRangeFloat(Minimum=8, Maximum=12)
    Index 3 Param 3=&RandomRangeFloat(Minimum=3, Maximum=5)
    Index 3 Param 4=&RandomRangeFloat(Minimum=0.5, Maximum=1)
    Module.Index0Float0Write=true
    Module.Index0Float1Write=true
    Module.Index0Float2Write=true
    Module.Index0Float3Write=true
    Module.Index1Float0Write=true
    Module.Index1Float1Write=true
    Module.Index1Float2Write=true
    Module.Index1Float3Write=true
    Module.Index2Float0Write=true
    Module.Index2Float1Write=true
    Module.Index2Float2Write=true
    Module.Index2Float3Write=true
    Module.Index3Float0Write=true
    Module.Index3Float1Write=true
    Module.Index3Float2Write=true
    Module.Index3Float3Write=true
    Write Parameter Index 1=::NewEnumerator1
    Write Parameter Index 2=::NewEnumerator1
    Write Parameter Index 3=::NewEnumerator1
   )
   &ScaleColor(
    Scale Alpha=&FloatFromCurve(
     FloatCurve=DataInterface() {propertyValues {Curve {defaultValue=1, keys [{tangentMode=RCTM_User, time=0, value=1}, {tangentMode=RCTM_User, time=1, value=0}]}}}
     CurveIndex=$Particles.NewNormalizedAge
     Scale Curve=&MakeCustomFloatFromBool(
      Bool=&SetBoolByFloatComparison(A=$Particles.NewNormalizedAge, B=1, Comparison Type=::NewEnumerator6)
      True Float=1
      False Float=0
     )
    )
    ScaleA=true
    ScaleRGB=false
   )
  ]
  rendererClasses [&NiagaraMeshRendererProperties, &NiagaraSpriteRendererProperties]
  renderers [
   {rendererClass=&NiagaraMeshRendererProperties, rendererIndex=0}
   {rendererClass=&NiagaraMeshRendererProperties, rendererIndex=1}
   {rendererClass=&NiagaraSpriteRendererProperties, rendererIndex=2}
   {rendererClass=&NiagaraMeshRendererProperties, rendererIndex=3}
  ]
 }
 ParaGlide_wave {
  emitterUpdateScript [
   &EmitterState(Life Cycle Mode=::NewEnumerator1, Loop Duration=$Engine.DeltaTime, Recalculate Duration Each Loop=true, UseLoopDelay=false)
   &SpawnBurst_Instantaneous(
    Spawn Probability=&MakeCustomFloatFromBool(Bool=$System.DoOnce.Execute, True Float=1, False Float=0)
    Use Spawn Probability=true
    UseLoopCountLimit=false
   ) {module=SpawnBurst_Instantaneous001}
  ]
  particleSpawnScript [
   &InitializeParticle(Lifetime=0.4, Mesh Scale Mode=::NewEnumerator3, Mesh Uniform Scale=0.6, UsePositionOffset=false, Write Material Random=true)
   &InitialMeshOrientation(
    Mesh Orientation Mode=::NewEnumerator2
    Orientation Vector=<0 1 0>
    Rotation=&RandomRangeVector(Minimum=<0 0 0>, Maximum=<0 0 1>)
    Use Rotation Vector=true
   )
   &AddVelocity(Velocity=<0 0 1>, Velocity Speed Scale=650)
   &SkeletalMeshLocation(
    Bone Sampling Mode=::NewEnumerator4
    MaskSpawnGroup=false
    Orientation Sampling=::NewEnumerator1
    Sampled Position Offset=<0 0 -65>
    Sampled Position Offset Coordinate Space:ENiagaraSkelMeshTransforms=::NewEnumerator1
    Skeletal Mesh=$User.SkeletalMesh
    Velocity Sampling=::NewEnumerator0
    VelocitySpeedLimit=false
   )
  ]
  particleUpdateScript [
   &ParticleState()
   &ScaleMeshSize(Scale Factor=&VectorFromFloat(Value=&FloatFromCurve(
      FloatCurve=DataInterface() {propertyValues {Curve {defaultValue=1, keys [{tangentMode=RCTM_User, time=0, value=1}, {tangentMode=RCTM_User, time=1, value=2}]}}}
      Scale Curve=1.25
     )))
   &ScaleColor(
    Scale Alpha=&FloatFromCurve(
     FloatCurve=DataInterface() {propertyValues {Curve {keys [
         {arriveTangent=26.1682, interpMode=RCIM_Cubic, leaveTangent=26.1682, tangentMode=RCTM_User, time=0, value=1}
         {arriveTangent=-3.7386, interpMode=RCIM_Cubic, leaveTangent=-3.7386, tangentMode=RCTM_User, time=0.1, value=1}
         {tangentMode=RCTM_User, time=1, value=0}
        ]}}}
     Scale Curve=2
    )
    ScaleA=true
    ScaleRGB=false
   )
   &SolveForcesAndVelocity(Clamp Velocity=false, Limit Acceleration=false)
  ]
  rendererClasses [&NiagaraMeshRendererProperties]
  renderers [{rendererClass=&NiagaraMeshRendererProperties, rendererIndex=0}]
 }
 Sparks {
  simTarget=GPUComputeSim
  emitterUpdateScript [
   &EmitterState(Life Cycle Mode=::NewEnumerator1, Loop Duration=$Engine.DeltaTime, UseLoopDelay=false)
   &SpawnBurst_Instantaneous(
    Spawn Count=120
    Spawn Probability=&MakeCustomFloatFromBool(Bool=$System.DoOnce.Execute, True Float=1, False Float=0)
    Use Spawn Probability=true
    UseLoopCountLimit=false
   )
   &SpawnRate(SpawnRate=&Lerp_Float(A=25, B=100, Alpha=$User.ParaGlideLeanAlpha), Use Spawn Probability=false)
  ]
  particleSpawnScript [
   &InitializeParticle(
    Color=<rgba 1 0.216 0.02 1>
    Lifetime=&Lerp_Float(
     A=&RandomRangeFloat(Minimum=0.4, Maximum=0.8)
     B=&RandomRangeFloat(Minimum=0.3, Maximum=0.6)
     Alpha=$User.ParaGlideLeanAlpha
    )
    Sprite Size=&Lerp_Vector2(
     A=&Multiply_Vector2D_ByFloat(
      Float=&MakeCustomFloatFromBool(
       Bool=$User.bIsFrontEndPreview
       True Float=&RandomRangeFloat(Minimum=1.5, Maximum=3)
       False Float=&RandomRangeFloat(Minimum=2, Maximum=4)
      )
      Vector2D=<1.5 4>
     )
     B=&Multiply_Vector2D_ByFloat(
      Float=&MakeCustomFloatFromBool(
       Bool=$User.bIsFrontEndPreview
       True Float=&RandomRangeFloat(Minimum=2, Maximum=5)
       False Float=&RandomRangeFloat(Minimum=2, Maximum=9)
      )
      Vector2D=<1.5 12>
     )
     Alpha=$User.ParaGlideLeanAlpha
    )
    Sprite Size Mode=::NewEnumerator4
    UsePositionOffset=true
    Write Material Random=true
   )
   &SkeletalMeshLocation(
    Bone Sampling Mode=::NewEnumerator4
    Interpolate Mesh Sampling=false
    MaskSpawnGroup=false
    Orientation Sampling=::NewEnumerator1
    Sampled Position Offset=&MakeCustomVectorFromBool(
     Bool=$User.bIsFrontEndPreview
     True Vector=&Lerp_Vector(
      A=&RandomRangeVector(Minimum=<-15 -15 -35>, Maximum=<15 15 -35>)
      B=&RandomRangeVector(Minimum=<-35 -35 -60>, Maximum=<35 35 -60>)
      Alpha=$User.ParaGlideLeanAlpha
     )
     False Vector=&Lerp_Vector(
      A=&RandomRangeVector(Minimum=<-25 -25 -40>, Maximum=<25 25 -40>)
      B=&RandomRangeVector(Minimum=<-45 -45 -50>, Maximum=<45 45 -50>)
      Alpha=$User.ParaGlideLeanAlpha
     )
    )
    Sampled Position Offset Coordinate Space:ENiagaraSkelMeshTransforms=::NewEnumerator2
    Skeletal Mesh=$User.SkeletalMesh
   )
   &AddVelocity(
    Constrain To Radius=1
    Module.Use Falloff Distance=false
    Origin Offset=&MakeCustomVectorFromBool(
     Bool=$User.bIsFrontEndPreview
     True Vector=&RandomRangeVector(Minimum=<-150 -150 0>, Maximum=<150 150 0>)
     False Vector=<1 0 0>
    )
    Override Seed=false
    Velocity Mode=::NewEnumerator1
    Velocity Speed=&Multiply_Float(
     A=&RandomRangeFloat(Minimum=100, Maximum=200)
     B=&Lerp_Float(
      A=1.25
      B=&MakeCustomFloatFromBool(Bool=$User.bIsFrontEndPreview, True Float=2.5, False Float=4)
      Alpha=$User.ParaGlideLeanAlpha
     )
    )
   ) {module=AddVelocity001}
  ]
  particleUpdateScript [
   &ParticleState()
   &ScaleSpriteSize(
    Scale Factor=&Lerp_Vector2(
     A=<1 1>
     B=&MakeVector2D(
      X=&FloatFromCurve(FloatCurve=DataInterface() {propertyValues {Curve {keys [
           {interpMode=RCIM_Cubic, tangentMode=RCTM_Break, time=0, value=1}
           {arriveTangent=-2.6568, interpMode=RCIM_Cubic, tangentMode=RCTM_Break, time=1, value=0}
          ]}}})
      Y=&FloatFromCurve(FloatCurve=DataInterface() {propertyValues {Curve {
          defaultValue=1
          keys [
           {interpMode=RCIM_Cubic, leaveTangent=-2.6242, tangentMode=RCTM_Break, time=0, value=1}
           {interpMode=RCIM_Cubic, tangentMode=RCTM_Break, time=1, value=0}
          ]
         }}})
     )
     Alpha=$User.ParaGlideLeanAlpha
    )
    Scale Sprite Size Mode=::NewEnumerator2
   )
   /Niagara/Modules/Update/Forces/CurlNoiseForce(
    Noise Frequency=100
    Noise Strength=&Multiply_Float(A=1000, B=&Lerp_Float(A=0.5, B=5, Alpha=$User.ParaGlideLeanAlpha))
    Pan Noise=true
   )
   &ScaleColor(
    Scale Alpha=&MakeCustomFloatFromBool(
     Bool=$User.bIsFrontEndPreview
     True Float=&FloatFromCurve(FloatCurve=DataInterface() {propertyValues {Curve {
         defaultValue=1
         keys [
          {interpMode=RCIM_Cubic, leaveTangent=1.4778, tangentMode=RCTM_Break, time=0, value=0}
          {interpMode=RCIM_Cubic, tangentMode=RCTM_Break, time=0.35, value=1.0147}
          {arriveTangent=-0.0405, interpMode=RCIM_Cubic, leaveTangent=-0.0405, tangentMode=RCTM_User, time=0.7009, value=1}
          {arriveTangent=-8.0089, interpMode=RCIM_Cubic, leaveTangent=-8.0089, tangentMode=RCTM_User, time=1, value=0}
         ]
        }}})
     False Float=&FloatFromCurve(FloatCurve=DataInterface() {propertyValues {Curve {
         defaultValue=1
         keys [
          {tangentMode=RCTM_User, time=0, value=1}
          {arriveTangent=-0.0405, interpMode=RCIM_Cubic, leaveTangent=-0.0405, tangentMode=RCTM_User, time=0.7009, value=1}
          {arriveTangent=-8.0089, interpMode=RCIM_Cubic, leaveTangent=-8.0089, tangentMode=RCTM_User, time=1, value=0}
         ]
        }}})
    )
    Scale RGB=&Lerp_Vector(
     A=&VectorFromFloat(Value=&Lerp_Float(
       A=4
       B=150
       Alpha=&FloatFromCurve(
        FloatCurve=DataInterface() {propertyValues {Curve {defaultValue=0, keys [<at 0 0 RCIM_Constant>, {tangentMode=RCTM_User, time=0.9, value=1}]}}}
        CurveIndex=&RandomRangeFloat(Minimum=0, Maximum=1)
       )
      ))
     B=<1 0 0>
     Alpha=$Particles.NormalizedAge
    )
    ScaleA=true
    ScaleRGB=true
   ) {module=ScaleColor001}
   &GravityForce(Gravity=&MakeCustomVectorFromBool(
     Bool=$User.bIsFrontEndPreview
     True Vector=&Lerp_Vector(A=<0 0 350>, B=<0 0 1500>, Alpha=$User.ParaGlideLeanAlpha)
     False Vector=<1 0 0>
    ))
   &SolveForcesAndVelocity(Clamp Velocity=false, Limit Acceleration=false)
  ]
  rendererClasses [&NiagaraSpriteRendererProperties]
  renderers [{rendererClass=&NiagaraSpriteRendererProperties}]
 }
 Speedlines {
  simTarget=GPUComputeSim
  emitterUpdateScript [
   &EmitterState(Life Cycle Mode=::NewEnumerator1, Loop Duration=1, UseLoopDelay=false)
   &SpawnRate(SpawnRate=&Lerp_Float(A=0, B=100, Alpha=$User.ParaGlideLeanAlpha), Use Spawn Probability=false) {module=SpawnRate001}
  ]
  particleSpawnScript [
   &InitializeParticle(
    Color=&MakeLinearColorFromVectorAndFloat(
     "Vector (RGB)"=<1 1 1>
     "Float (Alpha)"=&Multiply_Float(
      A=&Multiply_Float(
       A=&UniformRangedFloat(Minimum=0.015, Maximum=0.15)
       B=&FN_FrontEndFloatModifier(Gameplay=1, Front End=0.5)
      )
      B=&Add_Float(
       A=1
       B=&Multiply_Float(A=10, B=&Subtract_Float(A=1, B=$Engine.Emitter.SpawnCountScale))
      )
     )
    )
    Lifetime Max=0.3
    Lifetime Min=0.15
    Lifetime Mode=::NewEnumerator1
    Mass Mode=::NewEnumerator1
    Sprite Size=&UniformRangedVector2D(Minimum=<3 4>, Maximum=<5 6>)
    Sprite Size Mode=::NewEnumerator4
    UsePositionOffset=false
    Write Material Random=true
   ) {module=InitializeParticle001}
   &CylinderLocation(
    Cylinder Radius=&UniformRangedFloat(Minimum=100, Maximum=150)
    MaskSpawnGroup=false
    Override Seed=false
    UseOwnerScale=true
   )
   &AddVelocity(Velocity=&FN_FrontEndVector3Modifier(Gameplay=<0 0 1000>, Front End=<0 0 3000>))
   &AddVelocityFromPoint(
    Module.Use Falloff Distance=false
    Velocity Strength=&Multiply_Float(
     A=&FN_FrontEndFloatModifier(Gameplay=1, Front End=2)
     B=&UniformRangedFloat(Minimum=500, Maximum=700)
    )
   )
  ]
  particleUpdateScript [
   &ParticleState()
   &SolveForcesAndVelocity(Clamp Velocity=false, Limit Acceleration=false)
   &ScaleSpriteSize(Scale Factor=&Vector2DFromCurve(
     Vector2Curve=DataInterface() {propertyValues {XCurve {keys [<at 0 0.5 RCIM_Cubic>,<at 0.5 1 RCIM_Cubic>]}, YCurve {keys [<at 0 0 RCIM_Cubic>,<at 0.5 1 RCIM_Cubic>]}}}
     ScaleCurve=&Vector2DFromFloat(Value=1)
    ))
   &ScaleSpriteSizeBySpeed(Max Scale Factor=<1 40>, Min Scale Factor=<0.01 0.05>, Velocity Threshold=1)
   &ScaleColor(
    Scale Alpha=&FloatFromCurve(
     FloatCurve=DataInterface() {propertyValues {Curve {keys [<at 0 0 RCIM_Cubic>, <at 0.2 1>, {arriveTangent=-2.221, interpMode=RCIM_Cubic, leaveTangent=-2.221, tangentMode=RCTM_User, time=1, value=0}]}}}
     Scale Curve=&FN_FrontEndFloatModifier(Gameplay=&Lerp_Float(A=0, B=0.4, Alpha=$User.ParaGlideLeanAlpha), Front End=0)
    )
    ScaleA=true
    ScaleRGB=true
   )
  ]
  rendererClasses [&NiagaraSpriteRendererProperties]
  renderers [{rendererClass=&NiagaraSpriteRendererProperties}]
 }
}
summary {
 systemName=NS_Contrail_CampRock
 emitters [
  {emitterName=Speedlines, simTarget=GPUComputeSim, rendererClasses [&NiagaraSpriteRendererProperties]}
  {emitterName=Flame, rendererClasses [&NiagaraMeshRendererProperties]}
  {emitterName=Bonfire, rendererClasses [&NiagaraMeshRendererProperties, &NiagaraLightRendererProperties]}
  {emitterName=Hand_Trails, rendererClasses [&NiagaraMeshRendererProperties]}
  {emitterName=Sparks, simTarget=GPUComputeSim, rendererClasses [&NiagaraSpriteRendererProperties]}
  {emitterName=FireGlow, rendererClasses [&NiagaraSpriteRendererProperties]}
  {emitterName=ParaGlide_wave, rendererClasses [&NiagaraMeshRendererProperties]}
  {emitterName=Marshmallow, rendererClasses [&NiagaraMeshRendererProperties]}
  {
   emitterName=Marshmallow_withToasted001
   rendererClasses [&NiagaraMeshRendererProperties, &NiagaraSpriteRendererProperties]
  }
 ]
 userVariables [
  User.bIsFrontEnd {type=NiagaraBool}
  User.bIsFrontEndPreview {type=NiagaraBool}
  User.ParaGlideLeanAlpha {type=NiagaraFloat, defaultValue=0, description=""}
  User.PlaneHeight {type=NiagaraFloat, defaultValue=50, description=""}
  User.SkeletalMesh {
   type=NiagaraDataInterfaceSkeletalMesh
   defaultValue {
    struct=NiagaraExt_VariableValue_DataInterface
    value {dataInterface=&self:NiagaraDataInterfaceSkeletalMesh_0}
   }
  }
 ]
}
